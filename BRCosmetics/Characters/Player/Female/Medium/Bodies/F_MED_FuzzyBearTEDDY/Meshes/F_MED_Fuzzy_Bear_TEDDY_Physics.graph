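# Physics asset for the female medium Fuzzy Bear TEDDY player body cosmetic, covering two dynamic cloth bones at the right upper arm.
# Capsule bodies sit on the top cloth bone, bottom cloth bone and right upper arm.
# Each cloth bone is constrained directly to the arm rather than chained to the other, with moderate swing limits and a very tight twist limit.
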PhysicsAsset /BRCosmetics/Characters/Player/Female/Medium/Bodies/F_MED_FuzzyBearTEDDY/Meshes/F_MED_Fuzzy_Bear_TEDDY_Physics
bodies {
 dyn_cloth_btm_01 [{shapeType=Capsule, length=5.6186, radius=1.75, center=<-0.3809 -4.9677 -2.0048>, rotation=<rot 0 176.651 -72.8694>}]
 dyn_cloth_top_01 [{shapeType=Capsule, length=4.7992, radius=1.75, center=<0.6392 -3.7264 0.9723>, rotation=<rot 73.7344 153.4584 -127.6497>}]
 upperarm_r [{shapeType=Capsule, length=10.5256, radius=2.2941, center=<-14.1912 2.5914 -0.725>, rotation=<rot 0.6595 91.3485 -86.6143>}]
}
constraints [
 {bone1Name=dyn_cloth_top_01, bone2Name=upperarm_r, swing1LimitDegrees=60, swing2LimitDegrees=45, twistLimitDegrees=1}
 {bone1Name=dyn_cloth_btm_01, bone2Name=upperarm_r, swing1LimitDegrees=60, swing2LimitDegrees=45, twistLimitDegrees=1}
]
