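# Anim Blueprint for the female Fuzzy Bear TEDDY cosmetic body, copying the parent mesh pose and adding secondary motion and shoulder pad fixes.
# A pelvis-space RigidBody pass is cached, then shoulder pads get additive offsets weighted by upper-arm yaw, remapped each update.
# When isRenderScene is true, output skips physics and applies fixed cloth bone offsets instead.
# An isTargeting blend and an extra right-pad transform are unconnected.
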AnimBlueprint /BRCosmetics/Characters/Player/Female/Medium/Bodies/F_MED_FuzzyBearTEDDY/Meshes/F_MED_Fuzzy_Bear_TEDDY_AnimBP
target_skeleton=&/F_MED_Fuzzy_Bear_TEDDY_Skeleton
graphs [
 AnimGraph {nodes [
   Blend Poses by bool(BlendPose_1=Use cached pose 'AnimCache'(), bActiveValue=Get bIsTargeting()) {blendType=Linear}
   "Transform (Modify) Bone - Bone: dyn_shoulder_pad_r"(Alpha=1, Translation="-1.5,0,0", Rotation="0,0,0", Scale="1,1,1") {
    rotationMode=BMM_Additive
    translationMode=BMM_Additive
    alphaBoolBlend {blendOption=Linear}
    boneToModify {boneName=dyn_shoulder_pad_r}
    translation {x=-1.5}
   }
   n1 = "Transform (Modify) Bone - Bone: dyn_cloth_btm_01"(
    ComponentPose=Local To Component(LocalPose=Copy Pose From Mesh() {bUseAttachedParent=true})
    Alpha=1
    Translation="0,0,0"
    Rotation="0,0,0"
    Scale="1,1,1"
   ) {rotationMode=BMM_Additive, translationMode=BMM_Additive, alphaBoolBlend {blendOption=Linear}, boneToModify {boneName=dyn_cloth_btm_01}}
   n2 = "Transform (Modify) Bone - Bone: dyn_cloth_top_01"(ComponentPose=@n1, Alpha=1, Translation="0,0,0", Rotation="-20,0,-10", Scale="1,1,1") {
    rotationMode=BMM_Additive
    translationMode=BMM_Additive
    alphaBoolBlend {blendOption=Linear}
    boneToModify {boneName=dyn_cloth_top_01}
    rotation {pitch=-20, roll=-10}
   }
   n3 = "Transform (Modify) Bone - Bone: dyn_shoulder_pad_l"(
    ComponentPose=Local To Component(LocalPose=Use cached pose 'AnimCache'())
    Alpha=Get LshoulderAlpha() -> (LshoulderAlpha:double)
    Translation="0,0,6"
    Rotation="0,0,10"
    Scale="1,1,1"
   ) {
    rotationMode=BMM_Additive
    translationMode=BMM_Additive
    alphaBoolBlend {blendOption=Linear}
    boneToModify {boneName=dyn_shoulder_pad_l}
    rotation {roll=10}
    translation {z=6}
   }
   n4 = "Transform (Modify) Bone - Bone: dyn_shoulder_pad_r"(ComponentPose=@n3, Alpha=Get RshoulderAlpha() -> (RshoulderAlpha:double), Translation="-3,0,1", Rotation="0,0,0", Scale="1,1,1") {
    rotationMode=BMM_Additive
    translationMode=BMM_Additive
    alphaBoolBlend {blendOption=Linear}
    boneToModify {boneName=dyn_shoulder_pad_r}
    translation {x=-3, z=1}
   }
   Output Pose(Result=Blend Poses by bool(
     BlendPose_0=Component To Local(ComponentPose=@n2)
     BlendPose_1=Component To Local(ComponentPose=@n4)
     bActiveValue=Get isRenderScene() -> (isRenderScene:bool)
    ) {blendType=Linear}) {layerGroup=None}
   n5 = RigidBody(ComponentPose=Local To Component(LocalPose=Copy Pose From Mesh() {bUseAttachedParent=true}), Alpha=0.6) {
    bOverrideWorldGravity=true
    overridePhysicsAsset=&/F_MED_Fuzzy_Bear_TEDDY_Physics
    simulationSpace=BaseBoneSpace
    alphaBoolBlend {blendOption=Linear}
    baseBoneRef {boneName=pelvis}
    componentAppliedLinearAccClamp=<400 400 400>
    componentLinearAccScale=<5 5 5>
    componentLinearVelScale=<5 5 5>
    overrideWorldGravity {z=-980}
   }
   Save cached pose 'AnimCache'(Pose=Component To Local(ComponentPose=@n5)) {cacheName=AnimCache}
  ]}
 EventGraph {nodes [
   Event BlueprintUpdateAnimation()
   n1 = GetPartSkeletalMeshComponent()
   n2 = BreakTransform(InTransform=GetDeltaTransformFromRefPose(self=@n1, BoneName=upperarm_l, BaseName=spine_05))
   > Set LshoulderAlpha(LshoulderAlpha:double=MapRangeClamped(Value=BreakRotator(InRot=@n2.Rotation).Yaw, InRangeA=10, InRangeB=129)) -> (Output_Get:double)
   n3 = BreakTransform(InTransform=GetDeltaTransformFromRefPose(self=@n1, BoneName=upperarm_r, BaseName=clavicle_r))
   > Set RshoulderAlpha(RshoulderAlpha:double=MapRangeClamped(Value=BreakRotator(InRot=@n3.Rotation).Yaw, InRangeA=59, InRangeB=128)) -> (Output_Get:double)
   TryGetPawnOwner()
  ]}
]
variables [
 isRenderScene {type=boolean}
 LshoulderAlpha {type=number}
 RshoulderAlpha {type=number}
]
